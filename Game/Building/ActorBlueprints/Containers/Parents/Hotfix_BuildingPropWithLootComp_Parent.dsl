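; Parent Blueprint for lootable building props that can be switched off through hotfixable row values.
; Once playlist data is ready, it reads an enabled row and silently kills disabled actors after a random delay, staggering destruction without dropping loot.
; If only interaction is disabled, it blocks interact and can optionally clear the death-drop component's loot package.
; The loot event plays optional open sound and particle effect, with configurable offset and pitch.
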(event EventBeginPlay
  ;[1 1 1] On begin play check enable rows for interact and enabled. if we do not allow the interact also set loot table to null for destruction drops.
  (Class|FortAsyncActionCurrentPlaylistReady|PlaylistDataReadyAsync
    (:OnReady
      (if (Class|AthenaFunctionLibrary|RowtoBool (Variables|RowsParent|GetRow_ActorEnabled))
        (if (not (Class|AthenaFunctionLibrary|RowtoBool (Variables|RowsParent|GetRow_InteractEnabled)))
          (Variables|Interaction|SetbAllowInteract false)
          (if (Variables|RowsParent|GetDisableLootOnDisableInteract)
            (Variables|/Game/Athena/Items/Components/BuildingActorLootDropOnDeathComponent.BuildingActorLootDropOnDeathComponent_C|SetLootOnExplodePackage "None" (Variables|Default|GetBuildingActorLootDropOnDeathComponent))))
        (else
          (Utilities|FlowControl|Delay (Math|Random|RandomFloatinRange 0.1 3.0)) ; Give it a random before the kill so that we dont destroy on the same frame.
          (Gameplay|SilentDie false) ; Silent Die to not drop loot.
        )))))

(event Custom|OnLoot_1
  (bind _sound_open (Variables|SoundParent|GetSound_Open))
  (bind _fx_open (Variables|FXParent|GetFX_Open))
  (Utilities|FlowControl|Sequence
    (:then_0
      ;[1 1 1] Play Sound if Valid
      (Utilities|IsValid _sound_open
        (:Is%20Valid
          (Audio|PlaySoundatLocation _sound_open (Transformation|GetActorLocation)))))
    (:then_1
      ;[1 1 1] Play Open FX if valid
      (Utilities|IsValid _fx_open
        (:Is%20Valid
          (Effects|SpawnEmitteratLocation _fx_open (+ (Transformation|GetActorLocation) (Variables|FXParent|GetFX_OpenOffsetLocation)) (Math|Rotator|MakeRotator (.roll (Transformation|GetActorRotation)) (+ (.pitch (Transformation|GetActorRotation)) (Variables|FXParent|GetFX_OpenPitchOffset)) (.yaw (Transformation|GetActorRotation)))))))))

(fn ConstructionScript)
