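; Parent Blueprint for lootable building containers, handling playlist gating and opening feedback.
; When playlist data is ready, data rows can disable interaction and optionally clear loot, or silently kill the actor after a random delay.
; On loot, it optionally animates a material parameter, then fires a GameplayCue, falling back to sound, Niagara-or-Cascade FX and an authority-only minigame open event.
; A link-to-actor kludge destroys it when its support breaks.
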(event EventBeginPlay
  ;[1 1 1] On begin play check enable rows for interact and enabled. if we do not allow the interact also set loot table to null for destruction drops.
  (Class|FortAsyncActionCurrentPlaylistReady|PlaylistDataReadyAsync
    (:OnReady
      (if (Class|AthenaFunctionLibrary|RowtoBool (Variables|RowsParent|GetRow_ActorEnabled))
        (if (not (Class|AthenaFunctionLibrary|RowtoBool (Variables|RowsParent|GetRow_InteractEnabled)))
          (Variables|Interaction|SetbAllowInteract false)
          (if (Variables|RowsParent|GetDisableLootOnDisableInteract)
            (Variables|Container|SetSearchLootTierGroup "None")))
        (else
          (Utilities|FlowControl|Delay (Math|Random|RandomFloatinRange 0.1 3.0)) ; Give it a random before the kill so that we dont destroy on the same frame.
          (Gameplay|SilentDie false) ; Silent Die to not drop loot.
        )))))

(event EventOnLoot
  (bind (_location _rotation) (CallFunction|GetFXLocation))
  (bind (_location_2 _rotation_2) (CallFunction|GetFXLocation))
  (bind _gc_open (Variables|Open|GetGC_Open))
  (bind _sound_open (Variables|Open|GetSound_Open))
  ;[1 1 1] On Loot - Try to play the Open GC and also try to animate the container if it should
  (if (Variables|Open|GetContainerAnimates)
    (CallFunction|AnimateContainer)
    (else
      _))
  (if (GameplayTags|IsGameplayTagValid _gc_open)
    (GameplayCue|ExecuteGameplayCueOnActor(Burst) self _gc_open (Ability|GameplayCue|MakeGameplayCueParameters :NormalizedMagnitude 0.0 :RawMagnitude 0.0 :Location _location_2 :Normal (Math|Rotator|GetRotationXVector _rotation_2) :GameplayEffectLevel 1 :AbilityLevel 1 :bReplicateLocationWhenUsingMinimalRepProxy false))
    (else
      (Utilities|FlowControl|Sequence
        (:then_0
          ;[1 1 1] Play Sound if Valid
          (Utilities|IsValid _sound_open
            (:Is%20Valid
              (Audio|PlaySoundatLocation _sound_open (Transformation|GetActorLocation)))))
        (:then_1
          ;[1 1 1] Play Open FX if valid
          (CallFunction|PlayValidFx _location _rotation))
        (:then_2
          ;[0 1 0.9306 0.25] Check if we are in Creative Royale and trigger Open Event
          (Networking|SwitchHasAuthority
            (:Authority
              (Utilities|IsValid (Minigame|GetMinigame self :!class /Script/FortniteGame.MinigameBlueprintLibrary)
                (:Is%20Valid
                  (Class|LootControllerFunctionLibrary|MakeandSendOpenContainerEvent self (Variables|Default|GetContainer%20Type)))))))))))

(event Custom|AnimateContainer
  ;[1 1 1] Animate Container - Make a material instance then animate it using a timeline.
  (bind _returnvalue (Rendering|Material|CreateDynamicMaterialInstance (Variables|Debug|GetStaticMeshComponent) 0 0 "None" :!class /Script/Engine.PrimitiveComponent))
  (bind (_direction _time) (Timeline|Timeline Play)
    (:Update
      (Rendering|Material|SetScalarParameterValue _returnvalue (Variables|Open|GetOpenAnimationParameterName) _time :!class /Script/Engine.MaterialInstanceDynamic))))

(event OnLinkedActorDestroyed(FortLinkToActor)
  (params (DamageCauser :type Actor%20Object%20Reference))
  ;[1 0 0.05223] Sometimes the default actor linking for props is not linking as accurately as we would like. Using the link to actor component gets a better result. Destroy this actor if its support is broken. Kludge copied from Apollo_Farm_Cornbox_LootDrop; this should be replaced with more optimized improvement to B_ActorSpawner_Parent that doesn't require an extra component & Raycast
  (Actor|DestroyActor))

(fn GetFXLocation
  ;[1 1 1] Using the preset values for open Pitch and offet location, find the FX location and rotation.
  (return (+ (Transformation|GetActorLocation) (Variables|Open|GetFX_OpenOffsetLocation)) (Math|Rotator|MakeRotator (.roll (Transformation|GetActorRotation)) (+ (.pitch (Transformation|GetActorRotation)) (Variables|Open|GetFX_OpenPitchOffset)) (.yaw (Transformation|GetActorRotation)))))

(fn PlayValidFx
  (params (Location :type Vector) (Rotation :type Rotator))
  (bind _fx_niagaraopen (Variables|Open|GetFX_NiagaraOpen))
  (bind _fx_cascadeopen (Variables|Open|GetFX_CascadeOpen))
  ;[1 1 1] Try to play Niagara FX if those are invalid, Try to play Cascade ones.
  (Utilities|IsValid _fx_niagaraopen
    (:Is%20Valid
      (Niagara|SpawnSystematLocation _fx_niagaraopen Location Rotation))
    (:Is%20Not%20Valid
      (Utilities|IsValid _fx_cascadeopen
        (:Is%20Valid
          (Effects|SpawnEmitteratLocation _fx_cascadeopen Location Rotation))))))

(fn ConstructionScript)
